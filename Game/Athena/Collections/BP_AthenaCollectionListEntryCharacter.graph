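# Character tile widget for the Athena Collections list, with icon, number, "new" bang, completion banner and hover/selection feedback.
# Collection state sets bang visibility, number color and background material parameters; NewLocation and NewlyCompleted also style the banner, shown only if HighScoreAllowed.
# Hover and Select animations are driven by play functions whose Silent flag snaps material parameters and transforms to the end state instead of animating.
# Selection layers on hover, hides the bang, and entry reset deselects silently.
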WidgetBlueprint /Game/Athena/Collections/BP_AthenaCollectionListEntryCharacter
animations [
 Hover {
  binding_count=2
  length_seconds=0.15
  object=&self:Hover
  sequence {bindings [
    Border_Number {
     id=#1
     possessed_class=Border
     tracks [{
       type=MovieSceneColorTrack
       display_name=Content Color and Opacity
       sections [{
         type=MovieSceneColorSection
         channels [
          Color.R {type=MovieSceneScriptingFloatChannel}
          Color.G {type=MovieSceneScriptingFloatChannel}
          Color.B {type=MovieSceneScriptingFloatChannel}
          Color.A {type=MovieSceneScriptingFloatChannel, keys [[0,0.5],[3,0.8]]}
         ]
        }]
      }]
    }
    Image_Icon {
     id=#2
     possessed_class=FortLazyImage
     tracks [{
       type=MovieSceneWidgetMaterialTrack
       sections [{type=MovieSceneParameterSection, channels [hover {type=MovieSceneScriptingFloatChannel, keys [[0,0],[3,1]]}]}]
      }]
    }
   ]}
 }
 Select {
  binding_count=7
  length_seconds=0.15
  object=&self:Select
  sequence {bindings [
    Image_Icon {
     id=#3
     possessed_class=FortLazyImage
     tracks [{
       type=MovieSceneWidgetMaterialTrack
       sections [{type=MovieSceneParameterSection, channels [Selected {type=MovieSceneScriptingFloatChannel, keys [[0,0],[3,1]]}]}]
      }]
    }
    SizeBox_28 {
     id=#4
     possessed_class=SizeBox
     tracks [{
       type=MovieScene2DTransformTrack
       sections [{
         type=MovieScene2DTransformSection
         end_frame=3
         start_frame=0
         channels [
          Translation.X {type=MovieSceneScriptingFloatChannel}
          Translation.Y {type=MovieSceneScriptingFloatChannel}
          Angle {type=MovieSceneScriptingFloatChannel}
          Scale.X {type=MovieSceneScriptingFloatChannel, const [1.06,2]}
          Scale.Y {type=MovieSceneScriptingFloatChannel, keys [[0,1],[3,1.06]]}
          Shear.X {type=MovieSceneScriptingFloatChannel}
          Shear.Y {type=MovieSceneScriptingFloatChannel}
         ]
        }]
      }]
    }
    Image_SecondaryIcon {
     id=#5
     possessed_class=Image
     tracks [
      {
       type=MovieSceneWidgetMaterialTrack
       sections [{type=MovieSceneParameterSection, channels [Selected {type=MovieSceneScriptingFloatChannel, keys [[0,0],[3,1]]}]}]
      }
      {
       type=MovieScene2DTransformTrack
       sections [{
         type=MovieScene2DTransformSection
         end_frame=3
         start_frame=0
         channels [
          Translation.X {type=MovieSceneScriptingFloatChannel, keys [[0,0],[3,-6]]}
          Translation.Y {type=MovieSceneScriptingFloatChannel, keys [[0,0],[3,-6]]}
          Angle {type=MovieSceneScriptingFloatChannel}
          Scale.X {type=MovieSceneScriptingFloatChannel}
          Scale.Y {type=MovieSceneScriptingFloatChannel}
          Shear.X {type=MovieSceneScriptingFloatChannel}
          Shear.Y {type=MovieSceneScriptingFloatChannel}
         ]
        }]
      }
     ]
    }
    HighScoreText {
     id=#6
     possessed_class=CommonTextBlock
     tracks [{
       type=MovieSceneWidgetMaterialTrack
       display_name=Font.FontMaterial
       sections [{type=MovieSceneParameterSection, channels [HoverControl {type=MovieSceneScriptingFloatChannel, keys [[0,0],[3,1]]}]}]
      }]
    }
    HighScore {
     id=#7
     possessed_class=CommonBorder
     tracks [{
       type=MovieSceneWidgetMaterialTrack
       sections [{type=MovieSceneParameterSection, channels [HoverControl {type=MovieSceneScriptingFloatChannel, keys [[0,0],[3,1]]}]}]
      }]
    }
    Text_Number {
     id=#8
     possessed_class=CommonTextBlock
     tracks [{
       type=MovieScene2DTransformTrack
       sections [{
         type=MovieScene2DTransformSection
         end_frame=3
         start_frame=0
         channels [
          Translation.X {type=MovieSceneScriptingFloatChannel, keys [[0,0],[3,-4]]}
          Translation.Y {type=MovieSceneScriptingFloatChannel, keys [[0,0],[3,6]]}
          Angle {type=MovieSceneScriptingFloatChannel}
          Scale.X {type=MovieSceneScriptingFloatChannel}
          Scale.Y {type=MovieSceneScriptingFloatChannel}
          Shear.X {type=MovieSceneScriptingFloatChannel}
          Shear.Y {type=MovieSceneScriptingFloatChannel}
         ]
        }]
      }]
    }
    Image_selected {
     id=#9
     possessed_class=Image
     tracks [{
       type=MovieSceneWidgetMaterialTrack
       sections [{type=MovieSceneParameterSection, channels [hover {type=MovieSceneScriptingFloatChannel, keys [[0,0],[3,1]]}]}]
      }]
    }
   ]}
 }
]
graphs [
 EventGraph {nodes [
   Event On Entry Reset() {is=title K2Node_Event}
   > SetVisibility(self=Get NewBang() -> (NewBang:Common Border Object Reference), InVisibility=Collapsed)
   > SetVisibility(self=Get HighScore() -> (HighScore:Common Border Object Reference), InVisibility=Collapsed)
   > PlayDeselectionAnimation!(execute, self, Silent:bool=true) -> (then)
   Event On Hovered()
   n1 = IsUsingTouch(PlayerController=GetOwningPlayer())
   n2 = GetSelected!(self:Common Button Base Object Reference) -> (ReturnValue:bool)
   > Branch(Condition=AND Boolean(A=NOT Boolean(A=@n2), B=NOT Boolean(A=@n1)))
   > PlayHoverAnimation!(execute, self, Silent:bool=false) -> (then)
   Event On Unhovered()
   > n3 = Branch(Condition=GetSelected())
   PlayUnhoverAnimation!(execute=@n3.else, self, Silent:bool=false) -> (then)
   n4 = Event On Collection State Set!() -> (OutputDelegate, then, CollectedState:EFortCollectedState Enum=Unknown) {is=title K2Node_Event}
   > SetVisibility(
    self=Get NewBang()
    InVisibility=Select!(
     Unknown:ESlateVisibility Enum=Collapsed
     New:ESlateVisibility Enum=HitTestInvisible
     Known:ESlateVisibility Enum=Collapsed
     NewlyCollected:ESlateVisibility Enum=HitTestInvisible
     Collected:ESlateVisibility Enum=Collapsed
     NewBest:ESlateVisibility Enum=Collapsed
     NewRecord:ESlateVisibility Enum=Collapsed
     NewLocation:ESlateVisibility Enum=HitTestInvisible
     NewlyCompleted:ESlateVisibility Enum=HitTestInvisible
     Complete:ESlateVisibility Enum=Collapsed
     Index:EFortCollectedState Enum=@n4.CollectedState
    ) -> (ReturnValue)
   )
   n5 = Make Slate Color(SpecifiedColor=(R=0,G=0.3348,B=1,A=1))
   n6 = Make Slate Color(SpecifiedColor=(R=0,G=0.0578,B=0.6038,A=1))
   > SetColorAndOpacity(
    self=Get Text_Number() -> (Text_Number:Common Text Object Reference)
    InColorAndOpacity=Select(
     Unknown:Slate Color Structure=@n6
     New:Slate Color Structure=@n6
     Known:Slate Color Structure=Make Slate Color(SpecifiedColor=(R=0,G=0.1559,B=1,A=1))
     NewlyCollected:Slate Color Structure=@n5
     Collected:Slate Color Structure=@n5
     NewBest:Slate Color Structure=@n5
     NewRecord:Slate Color Structure=@n5
     NewLocation:Slate Color Structure=@n5
     NewlyCompleted:Slate Color Structure=@n5
     Complete:Slate Color Structure=@n5
     Index=@n4.CollectedState
    ) -> (ReturnValue:Slate Color Structure)
   )
   n7 = Select!(
    Option 0:ESlateVisibility Enum=Collapsed
    Option 1:ESlateVisibility Enum=HitTestInvisible
    Index:bool=Get HighScoreAllowed() -> (HighScoreAllowed:bool)
   ) -> (ReturnValue:ESlateVisibility Enum)
   > SetVisibility(
    self=Get HighScore()
    InVisibility=Select!(
     Unknown:ESlateVisibility Enum=Collapsed
     New:ESlateVisibility Enum=Collapsed
     Known:ESlateVisibility Enum=Collapsed
     NewlyCollected:ESlateVisibility Enum=Collapsed
     Collected:ESlateVisibility Enum=Collapsed
     NewBest:ESlateVisibility Enum=Collapsed
     NewRecord:ESlateVisibility Enum=Collapsed
     NewLocation:ESlateVisibility Enum=@n7
     NewlyCompleted:ESlateVisibility Enum=@n7
     Complete:ESlateVisibility Enum=Collapsed
     Index:EFortCollectedState Enum=@n4.CollectedState
    ) -> (ReturnValue)
   )
   > n8 = GetDynamicMaterial(self=Get Image_Background() -> (Image_Background:Image Object Reference))
   > SetScalarParameterValue(
    self=@n8
    ParameterName=IsKnown
    Value=Select!(
     Option 0:double=1
     Option 1:double=0
     Index:bool="Equal (Enum)"(A:EFortCollectedState Enum=@n4.CollectedState, B:EFortCollectedState Enum=Unknown)
    ) -> (ReturnValue:double)
   )
   > SetScalarParameterValue(
    self=@n8
    ParameterName=IsDiscovered
    Value=Select!(
     Unknown:double=0
     New:double=0
     Known:double=0
     NewlyCollected:double=1
     Collected:double=1
     NewBest:double=1
     NewRecord:double=1
     NewLocation:double=1
     NewlyCompleted:double=1
     Complete:double=1
     Index:EFortCollectedState Enum=@n4.CollectedState
    ) -> (ReturnValue)
   )
   n9 = "Equal (Enum)"(A=@n4.CollectedState, B=NewLocation)
   > Branch(Condition=OR Boolean(A=@n9, B="Equal (Enum)"(A=@n4.CollectedState, B=NewlyCompleted)))
   > n10 = GetDynamicMaterial(self:Border Object Reference=Get HighScore())
   > SetVectorParameterValue(
    self=@n10
    ParameterName=MainColor
    Value=Select!(
     Option 0:Linear Color Structure=Get CompleteBorderColor() -> (CompleteBorderColor:Linear Color Structure)
     Option 1:Linear Color Structure=Get NewLocationBorderColor() -> (NewLocationBorderColor:Linear Color Structure)
     Index:bool=@n9
    ) -> (ReturnValue:Linear Color Structure)
   )
   n11 = Get HighScoreText() -> (HighScoreText:Common Text Object Reference)
   > "SetText (Text)"(
    self=@n11
    InText=Select(Option 0:Text=Get CompleteText() -> (CompleteText:Text), Option 1:Text=Get NewLocationText() -> (NewLocationText:Text), Index=@n9) -> (ReturnValue:Text)
   )
   > n12 = GetDynamicFontMaterial!(execute, self:Text Object Reference=@n11) -> (then, ReturnValue:Material Instance Dynamic Object Reference)
   n13 = Get NewLocationTextColor() -> (NewLocationTextColor:Slate Color Structure)
   n14 = Get CompleteTextColor() -> (CompleteTextColor:Slate Color Structure)
   > SetVectorParameterValue(
    self=@n12
    ParameterName=MainColor
    Value=Select(
     Option 0:Linear Color Structure=Break Slate Color!(SlateColor:"Slate Color Structure (by ref)"=@n14) -> (SpecifiedColor:Linear Color Structure, ColorUseRule:ESlateColorStylingMode Enum) {is=title K2Node_BreakStruct}
     Option 1:Linear Color Structure=Break Slate Color(SlateColor=@n13)
     Index=@n9
    ) -> (ReturnValue:Linear Color Structure)
   )
   Event On Selected()
   > n15 = PlaySelectionAnimation!(execute, self, Silent:bool=Get bResetInProgress() -> (bResetInProgress:bool)) -> (then)
   Branch(execute=[@n16.then, @n15.then], Condition=GetSelected())
   > SetVisibility(self=Get NewBang(), InVisibility=Collapsed)
   Event On Deselected()
   > n16 = PlayDeselectionAnimation(Silent=Get bResetInProgress())
   Event On Entry Image Set() {is=title K2Node_Event}
  ]}
 PlayDeselectionAnimation {nodes [
   n1 = PlayDeselectionAnimation!() -> (then, Silent:bool=~) {is=title K2Node_FunctionEntry}
   > StopAnimation(InAnimation=Get Select() -> (Select:Widget Animation Object Reference))
   > PlayUnhoverAnimation(Silent=@n1)
   > n2 = Sequence()
   > n3 = Branch(Condition=@n1)
   > n4 = GetDynamicMaterial(self=Get HighScore())
   > SetScalarParameterValue(self=@n4, ParameterName=HoverControl, Value=0)
   > n5 = GetDynamicFontMaterial(self=Get HighScoreText())
   > SetScalarParameterValue(self=@n5, ParameterName=HoverControl, Value=0)
   > n6 = GetDynamicMaterial(self:Image Object Reference=Get Image_Icon() -> (Image_Icon:Image Object Reference))
   > SetScalarParameterValue(self=@n6, ParameterName=Selected, Value=0)
   n7 = Get Image_SecondaryIcon() -> (Image_SecondaryIcon:Image Object Reference)
   > n8 = GetDynamicMaterial(self=@n7)
   > SetScalarParameterValue(self=@n8, ParameterName=Selected, Value=0)
   > SetRenderTransform!(
    execute
    self:Widget Object Reference=@n7
    InTransform_Translation:Vector 2D Structure=(X=0,Y=0)
    InTransform_Scale:Vector 2D Structure=(X=1,Y=1)
    InTransform_Shear:Vector 2D Structure=(X=0,Y=0)
    InTransform_Angle:float=0
   ) -> (then)
   > n9 = GetDynamicMaterial(self=Get Image_selected() -> (Image_selected:Image Object Reference))
   > SetScalarParameterValue(self=@n9, ParameterName=hover, Value=0)
   > SetRenderTransform(
    self=Get SizeBox_28() -> (SizeBox_28:Size Box Object Reference)
    InTransform_Translation=(X=0,Y=0)
    InTransform_Scale=(X=1,Y=1)
    InTransform_Shear=(X=0,Y=0)
    InTransform_Angle=0
   )
   > SetRenderTransform(self=Get Text_Number(), InTransform_Translation=(X=0,Y=0), InTransform_Scale=(X=1,Y=1), InTransform_Shear=(X=0,Y=0), InTransform_Angle=0)
   Branch(execute=@n3.else, Condition=Get Selected() -> (Selected:bool))
   > PlayAnimationReverse(InAnimation=Get Select())
   Set Selected(execute=@n2.then_1, Selected:bool=false) -> (Output_Get:bool)
  ]}
 PlayHoverAnimation {nodes [
   n1 = PlayHoverAnimation!() -> (then, Silent:bool=~) {is=title K2Node_FunctionEntry}
   > StopAnimation(InAnimation=Get Hover() -> (Hover:Widget Animation Object Reference))
   > n2 = Sequence()
   > n3 = Branch(Condition=@n1)
   > Set ContentColorAndOpacity!(
    execute
    ContentColorAndOpacity:Linear Color Structure=(R=1,G=1,B=1,A=0.8)
    self:Border Object Reference=Get Border_Number() -> (Border_Number:Border Object Reference)
   ) -> (then, Output_Get:Linear Color Structure)
   > n4 = GetDynamicMaterial(self=Get Image_Icon())
   > SetScalarParameterValue(self=@n4, ParameterName=hover, Value=1)
   n5 = Branch(execute=@n3.else, Condition=Get Hovered() -> (Hovered:bool))
   PlayAnimationForward(execute=@n5.else, InAnimation=Get Hover())
   Set Hovered(execute=@n2.then_1, Hovered:bool=true) -> (Output_Get:bool)
  ]}
 PlaySelectionAnimation {nodes [
   n1 = PlaySelectionAnimation!() -> (then, Silent:bool=~) {is=title K2Node_FunctionEntry}
   > StopAnimation(InAnimation=Get Select())
   > PlayHoverAnimation(Silent=@n1) {is=title K2Node_CallFunction}
   > n2 = Sequence()
   > n3 = Branch(Condition=@n1)
   > n4 = GetDynamicMaterial(self:Border Object Reference=Get HighScore())
   > SetScalarParameterValue(self=@n4, ParameterName=HoverControl, Value=1)
   > n5 = GetDynamicFontMaterial(self=Get HighScoreText())
   > SetScalarParameterValue(self=@n5, ParameterName=HoverControl, Value=1)
   > n6 = GetDynamicMaterial(self:Image Object Reference=Get Image_Icon())
   > SetScalarParameterValue(self=@n6, ParameterName=Selected, Value=1)
   n7 = Get Image_SecondaryIcon()
   > n8 = GetDynamicMaterial(self=@n7)
   > SetScalarParameterValue(self=@n8, ParameterName=Selected, Value=1)
   > SetRenderTransform(self=@n7, InTransform_Translation=(X=-6,Y=-6), InTransform_Scale=(X=1,Y=1), InTransform_Shear=(X=0,Y=0), InTransform_Angle=0)
   > n9 = GetDynamicMaterial(self=Get Image_selected())
   > SetScalarParameterValue(self=@n9, ParameterName=hover, Value=1)
   > SetRenderTransform(
    self=Get SizeBox_28()
    InTransform_Translation=(X=0,Y=0)
    InTransform_Scale=(X=1.06,Y=1.06)
    InTransform_Shear=(X=0,Y=0)
    InTransform_Angle=0
   )
   > SetRenderTransform(self=Get Text_Number(), InTransform_Translation=(X=-4,Y=6), InTransform_Scale=(X=1,Y=1), InTransform_Shear=(X=0,Y=0), InTransform_Angle=0)
   n10 = Branch(execute=@n3.else, Condition=Get Selected())
   PlayAnimationForward(execute=@n10.else, InAnimation=Get Select())
   Set Selected(execute=@n2.then_1, Selected=true)
  ]}
 PlayUnhoverAnimation {nodes [
   n1 = StopAnimation(InAnimation=Get Hover())
   n2 = Sequence(execute=[@n1.then, @n3.then])
   n3 = PlayUnhoverAnimation!() -> (then, Silent:bool=~) {is=title K2Node_FunctionEntry}
   n4 = Branch(execute=@n2.then_0, Condition=@n3)
   > Set ContentColorAndOpacity(ContentColorAndOpacity=(R=1,G=1,B=1,A=0.5), self=Get Border_Number())
   > n5 = GetDynamicMaterial(self=Get Image_Icon())
   > SetScalarParameterValue(self=@n5, ParameterName=hover, Value=0)
   Branch(execute=@n4.else, Condition=Get Hovered())
   > PlayAnimationReverse(InAnimation=Get Hover())
   Set Hovered(execute=@n2.then_1, Hovered=false)
  ]}
 Update Hover Parameter {nodes [
   Update Hover Parameter() {is=title K2Node_FunctionEntry}
   > n1 = GetDynamicMaterial(self=Get Image_Icon())
   n2 = GetSelected()
   > SetScalarParameterValue(self=@n1, ParameterName=hover, Value=SelectFloat(A=1, B=0, bPickA=@n2))
   > SetRenderTranslation(
    self=Get Text_Number()
    Translation=Select(Option 0:Vector 2D Structure, Option 1:Vector 2D Structure=(X=-4,Y=6), Index=@n2) -> (ReturnValue:Vector 2D Structure)
   )
   > SetContentColorAndOpacity!(
    execute
    self:Border Object Reference=Get Border_Number()
    InContentColorAndOpacity:Linear Color Structure=Select(Option 0:Linear Color Structure=(R=1,G=1,B=1,A=0.33), Option 1:Linear Color Structure=(R=1,G=1,B=1,A=0.66), Index=@n2) -> (ReturnValue:Linear Color Structure)
   ) -> (then)
   > n3 = GetDynamicMaterial(self=Get Image_Icon())
   > SetScalarParameterValue(self=@n3, ParameterName=Selected, Value=Select(Option 0:double=0, Option 1:double=1, Index=@n2) -> (ReturnValue:double))
   > SetRenderScale!(
    execute
    self:Widget Object Reference=Get SizeBox_28()
    Scale:Vector 2D Structure=Select(Option 0:Vector 2D Structure=(X=1,Y=1), Option 1:Vector 2D Structure=(X=1.06,Y=1.06), Index=@n2) -> (ReturnValue:Vector 2D Structure)
   ) -> (then)
   > n4 = GetDynamicMaterial(self:Border Object Reference=Get HighScore())
   n5 = Select(Option 0:double=0, Option 1:double=1, Index=@n2) -> (ReturnValue:double)
   > SetScalarParameterValue(self=@n4, ParameterName=HoverControl, Value=@n5)
   > n6 = GetDynamicFontMaterial(self=Get HighScoreText())
   > SetScalarParameterValue(self=@n6, ParameterName=HoverControl, Value=@n5)
   > Set Hovered(Hovered=@n2)
   > Set Selected(Selected=@n2)
  ]}
]
variables [
 CompleteBorderColor {type=LinearColor, default=<rgba 0.9301 1 0.0097 1>}
 CompleteText {type=string, default="complete!"}
 CompleteTextColor {type=SlateColor, default {colorUseRule=UseColor_Specified, specifiedColor=<rgba 0 0 0 1>}}
 HighScoreAllowed {type=boolean}
 Hovered {type=boolean}
 NewLocationBorderColor {type=LinearColor, default=<rgba 0.7913 0.008 0.1845 1>}
 NewLocationText {type=string, default="New location!"}
 NewLocationTextColor {type=SlateColor, default {colorUseRule=UseColor_Specified, specifiedColor=<rgba 1 1 1 1>}}
 Selected {type=boolean}
]
widgets {
 info {
  inheritedWidgetCount=6
  parentClass=/Script/FortniteUI.AthenaCollectionListEntryCharacter
  rootWidgetClass=&SizeBox
  widgetCount=20
 }
 widgets [
  SizeBox_28 {
   widgetClassPath=&SizeBox
   bIsVariable=true
   widget=SizeBox_28
   properties {bOverride_HeightOverride=true, bOverride_WidthOverride=true, heightOverride=160, widthOverride=128}
  }
  Overlay_39 {widgetClassPath=&Overlay, parent=SizeBox_28, widget=Overlay_39, slot {class=SizeBoxSlot}}
  Image_Background {
   widgetClassPath=&Image
   bInherited=true
   bIsVariable=true
   parent=Overlay_39
   widget=Image_Background
   properties {brush {
     bIsDynamicallyLoaded=false
     drawAs=Image
     imageType=NoImage
     mirroring=NoMirror
     resourceName=None
     resourceObject=/Game/Athena/Collections/Art/Character/M_Ui_CharacterCollectionEntry_Background_Inst
     tiling=NoTile
     imageSize=<32 32>
     margin=<ltrb 0 0 0 0>
     outlineSettings {
      bUseBrushTransparency=false
      roundingType=HalfHeightRadius
      width=0
      color {colorUseRule=UseColor_Specified, specifiedColor=<rgba 0 0 0 0>}
      cornerRadii=<0 0 0 0>
     }
     tintColor {colorUseRule=UseColor_Specified, specifiedColor=<rgba 1 1 1 1>}
     uVRegion {bIsValid=false, max=<0 0>, min=<0 0>}
    }}
   slot {class=OverlaySlot, properties {horizontalAlignment=HAlign_Fill, verticalAlignment=VAlign_Fill}}
  }
  Border_Number {
   widgetClassPath=&Border
   bIsVariable=true
   parent=Overlay_39
   widget=Border_Number
   properties {
    background {
     bIsDynamicallyLoaded=false
     drawAs=NoDrawType
     imageType=NoImage
     mirroring=NoMirror
     resourceName=None
     resourceObject=null
     tiling=NoTile
     imageSize=<32 32>
     margin=<ltrb 0 0 0 0>
     outlineSettings {
      bUseBrushTransparency=false
      roundingType=HalfHeightRadius
      width=0
      color {colorUseRule=UseColor_Specified, specifiedColor=<rgba 0 0 0 0>}
      cornerRadii=<0 0 0 0>
     }
     tintColor {colorUseRule=UseColor_Specified, specifiedColor=<rgba 1 1 1 1>}
     uVRegion {bIsValid=false, max=<0 0>, min=<0 0>}
    }
    contentColorAndOpacity=<rgba 1 1 1 0.5>
   }
   slot {class=OverlaySlot, properties {horizontalAlignment=HAlign_Right, padding=<ltrb 0 -4 4 0>}}
  }
  Text_Number {
   widgetClassPath=&CommonTextBlock
   bInherited=true
   parent=Border_Number
   widget=Text_Number
   properties {
    text=01
    colorAndOpacity {colorUseRule=UseColor_Specified, specifiedColor=<rgba 0 0.1559 1 1>}
    font {
     bForceMonospaced=false
     bMaterialIsStencil=false
     fontMaterial=null
     fontObject=&Burbank
     letterSpacing=0
     monospacedWidth=1
     size=24
     skewAmount=0
     typefaceFontName=BurbankRegular-Black
     outlineSettings {
      bApplyOutlineToDropShadows=false
      bMiteredCorners=false
      bSeparateFillAlpha=false
      outlineMaterial=null
      outlineSize=0
      outlineColor=<rgba 0 0 0 1>
     }
    }
    renderTransform {angle=0, scale=<1 1>, shear=<-12 0>, translation=<0 0>}
    shadowOffset=<0 0>
   }
   slot {class=BorderSlot}
  }
  Image_Icon {
   widgetClassPath=&FortLazyImage
   bInherited=true
   bIsVariable=true
   parent=Overlay_39
   widget=Image_Icon
   properties {
    visibility=Visible
    brush {
     bIsDynamicallyLoaded=false
     drawAs=Image
     imageType=NoImage
     mirroring=NoMirror
     resourceName=None
     resourceObject=/Game/Athena/Collections/Art/Character/M_Ui_CharacterCollectionEntry_Icon_Inst
     tiling=NoTile
     imageSize=<32 32>
     margin=<ltrb 0 0 0 0>
     outlineSettings {
      bUseBrushTransparency=false
      roundingType=HalfHeightRadius
      width=0
      color {colorUseRule=UseColor_Specified, specifiedColor=<rgba 0 0 0 0>}
      cornerRadii=<0 0 0 0>
     }
     tintColor {colorUseRule=UseColor_Specified, specifiedColor=<rgba 1 1 1 1>}
     uVRegion {bIsValid=false, max=<0 0>, min=<0 0>}
    }
   }
   slot {class=OverlaySlot, properties {horizontalAlignment=HAlign_Fill, verticalAlignment=VAlign_Fill}}
  }
  Image_SecondaryIcon {
   widgetClassPath=&Image
   bInherited=true
   bIsVariable=true
   parent=Overlay_39
   widget=Image_SecondaryIcon
   properties {
    visibility=Collapsed
    brush {
     bIsDynamicallyLoaded=false
     drawAs=Image
     imageType=NoImage
     mirroring=NoMirror
     resourceName=None
     resourceObject=null
     tiling=NoTile
     imageSize=<48 48>
     margin=<ltrb 0 0 0 0>
     outlineSettings {
      bUseBrushTransparency=false
      roundingType=HalfHeightRadius
      width=0
      color {colorUseRule=UseColor_Specified, specifiedColor=<rgba 0 0 0 0>}
      cornerRadii=<0 0 0 0>
     }
     tintColor {colorUseRule=UseColor_Specified, specifiedColor=<rgba 1 1 1 1>}
     uVRegion {bIsValid=false, max=<0 0>, min=<0 0>}
    }
   }
   slot {class=OverlaySlot, properties {horizontalAlignment=HAlign_Right, verticalAlignment=VAlign_Bottom}}
  }
  Image_selected {
   widgetClassPath=&Image
   bIsVariable=true
   parent=Overlay_39
   widget=Image_selected
   properties {brush {
     bIsDynamicallyLoaded=false
     drawAs=Image
     imageType=NoImage
     mirroring=NoMirror
     resourceName=None
     resourceObject=/Game/Athena/Collections/Art/Fishing/M_Ui_CollectionEntry_Hover_Inst
     tiling=NoTile
     imageSize=<32 32>
     margin=<ltrb 0 0 0 0>
     outlineSettings {
      bUseBrushTransparency=false
      roundingType=HalfHeightRadius
      width=0
      color {colorUseRule=UseColor_Specified, specifiedColor=<rgba 0 0 0 0>}
      cornerRadii=<0 0 0 0>
     }
     tintColor {colorUseRule=UseColor_Specified, specifiedColor=<rgba 1 1 1 1>}
     uVRegion {bIsValid=false, max=<0 0>, min=<0 0>}
    }}
   slot {class=OverlaySlot, properties {horizontalAlignment=HAlign_Fill, verticalAlignment=VAlign_Fill}}
  }
  HighScore {
   widgetClassPath=&CommonBorder
   bIsVariable=true
   parent=Overlay_39
   widget=HighScore
   properties {
    visibility=Collapsed
    background {
     bIsDynamicallyLoaded=false
     drawAs=Box
     imageType=NoImage
     mirroring=NoMirror
     resourceName=None
     resourceObject=/Game/Athena/UI/ReplayBrowser/M_Ui_Skewed_Vfor9slices_Hover
     tiling=NoTile
     imageSize=<32 32>
     margin=<ltrb 0 0 0 0>
     outlineSettings {
      bUseBrushTransparency=false
      roundingType=HalfHeightRadius
      width=0
      color {colorUseRule=UseColor_Specified, specifiedColor=<rgba 0 0 0 0>}
      cornerRadii=<0 0 0 0>
     }
     tintColor {colorUseRule=UseColor_Specified, specifiedColor=<rgba 1 1 1 1>}
     uVRegion {bIsValid=false, max=<0 0>, min=<0 0>}
    }
    padding=<ltrb 8 3 8 1>
   }
   slot {class=OverlaySlot, properties {horizontalAlignment=HAlign_Fill, verticalAlignment=VAlign_Bottom}}
  }
  HighScoreText {
   widgetClassPath=&CommonTextBlock
   bIsVariable=true
   parent=HighScore
   widget=HighScoreText
   properties {
    justification=Center
    mobileFontSizeMultiplier=1.2
    scrollStyle=&TextScrollStyle-Base_C
    text="complete!"
    textTransformPolicy=ToUpper
    font {
     bForceMonospaced=false
     bMaterialIsStencil=false
     fontMaterial=/Game/Athena/Collections/Art/M_Ui_textColorHover
     fontObject=&BurbankSmall
     letterSpacing=0
     monospacedWidth=1
     size=12
     skewAmount=0
     typefaceFontName=Black
     outlineSettings {
      bApplyOutlineToDropShadows=false
      bMiteredCorners=false
      bSeparateFillAlpha=false
      outlineMaterial=null
      outlineSize=0
      outlineColor=<rgba 0 0 0 1>
     }
    }
    renderTransform {angle=0, scale=<1 1>, shear=<-12 -1.5>, translation=<0 0>}
    shadowOffset=<0 0>
   }
   slot {class=BorderSlot, properties {padding=<ltrb 8 3 8 1>}}
  }
  NewBang {
   widgetClassPath=&CommonBorder
   bIsVariable=true
   parent=Overlay_39
   widget=NewBang
   properties {
    horizontalAlignment=HAlign_Center
    background {
     bIsDynamicallyLoaded=false
     drawAs=Box
     imageType=NoImage
     mirroring=NoMirror
     resourceName=None
     resourceObject=/Game/Athena/UI/ReplayBrowser/M_Ui_Skewed_Vfor9slices_Bang
     tiling=NoTile
     imageSize=<32 32>
     margin=<ltrb 0.5 0 0.5 0>
     outlineSettings {
      bUseBrushTransparency=false
      roundingType=HalfHeightRadius
      width=0
      color {colorUseRule=UseColor_Specified, specifiedColor=<rgba 0 0 0 0>}
      cornerRadii=<0 0 0 0>
     }
     tintColor {colorUseRule=UseColor_Specified, specifiedColor=<rgba 0.9301 1 0.0097 1>}
     uVRegion {bIsValid=false, max=<0 0>, min=<0 0>}
    }
    padding=<ltrb 8 6 8 6>
   }
   slot {class=OverlaySlot, properties {horizontalAlignment=HAlign_Right, padding=<ltrb 4 -6 4 4>}}
  }
  CommonTextBlock_209 {
   widgetClassPath=&CommonTextBlock
   parent=NewBang
   widget=CommonTextBlock_209
   properties {
    lineHeightPercentage=0.7429
    text="!"
    colorAndOpacity {colorUseRule=UseColor_Specified, specifiedColor=<rgba 0 0 0 1>}
    font {
     bForceMonospaced=false
     bMaterialIsStencil=false
     fontMaterial=null
     fontObject=&Burbank
     letterSpacing=0
     monospacedWidth=1
     size=18
     skewAmount=0
     typefaceFontName=BurbankRegular-Black
     outlineSettings {
      bApplyOutlineToDropShadows=false
      bMiteredCorners=false
      bSeparateFillAlpha=false
      outlineMaterial=null
      outlineSize=0
      outlineColor=<rgba 0.9301 1 0.0097 1>
     }
    }
    margin=<ltrb 0 -6 0 0>
    renderTransform {angle=0, scale=<1 1>, shear=<-12 0>, translation=<0 0>}
   }
   slot {class=BorderSlot, properties {horizontalAlignment=HAlign_Center, padding=<ltrb 8 6 8 6>}}
  }
  ItemNameAttachment {
   widgetClassPath=&FortVisualAttachment
   bIsVariable=true
   parent=Overlay_39
   widget=ItemNameAttachment
   properties {maxDesiredHeight=58.5233, visibility=Collapsed, widthOverride=150, contentAnchor=<0.5 1>}
   slot {class=OverlaySlot, properties {horizontalAlignment=HAlign_Center}}
  }
  VerticalBox_0 {
   widgetClassPath=&VerticalBox
   parent=ItemNameAttachment
   widget=VerticalBox_0
   properties {renderTransformPivot=<0.5 1>}
   slot {class=SizeBoxSlot, properties {padding=<ltrb 0 0 0 1>}}
  }
  SizeBox_2 {
   widgetClassPath=&SizeBox
   parent=VerticalBox_0
   widget=SizeBox_2
   properties {bOverride_MaxDesiredWidth=true, bOverride_MinDesiredWidth=true, maxDesiredWidth=200, minDesiredWidth=140}
   slot {class=VerticalBoxSlot}
  }
  InvalidationBox_0 {widgetClassPath=/Script/UMG.InvalidationBox, parent=SizeBox_2, widget=InvalidationBox_0, slot {class=SizeBoxSlot}}
  CommonBorder_1 {
   widgetClassPath=&CommonBorder
   parent=InvalidationBox_0
   widget=CommonBorder_1
   properties {
    verticalAlignment=VAlign_Center
    background {
     bIsDynamicallyLoaded=false
     drawAs=Box
     imageType=NoImage
     mirroring=NoMirror
     resourceName=None
     resourceObject=/Game/Athena/UI/ReplayBrowser/M_Ui_Skewed_NoV_Nobottom
     tiling=NoTile
     imageSize=<32 32>
     margin=<ltrb 0.25 0 0.25 0>
     outlineSettings {
      bUseBrushTransparency=false
      roundingType=HalfHeightRadius
      width=0
      color {colorUseRule=UseColor_Specified, specifiedColor=<rgba 0 0 0 0>}
      cornerRadii=<0 0 0 0>
     }
     tintColor {colorUseRule=UseColor_Specified, specifiedColor=<rgba 1 1 1 1>}
     uVRegion {bIsValid=false, max=<0 0>, min=<0 0>}
    }
    padding=<ltrb 6 6 6 6>
   }
   slot {class=PanelSlot}
  }
  Text_ItemName {
   widgetClassPath=&CommonTextBlock
   bInherited=true
   parent=CommonBorder_1
   widget=Text_ItemName
   properties {
    autoWrapText=true
    clipping=ClipToBounds
    justification=Center
    lineHeightPercentage=0.9
    mobileFontSizeMultiplier=1.2
    text=Very Long Item Name
    textTransformPolicy=ToUpper
    colorAndOpacity {colorUseRule=UseColor_Specified, specifiedColor=<rgba 0 0 0 1>}
    font {
     bForceMonospaced=false
     bMaterialIsStencil=false
     fontMaterial=null
     fontObject=&BurbankSmall
     letterSpacing=0
     monospacedWidth=1
     size=14
     skewAmount=0
     typefaceFontName=Black
     outlineSettings {
      bApplyOutlineToDropShadows=false
      bMiteredCorners=false
      bSeparateFillAlpha=false
      outlineMaterial=null
      outlineSize=0
      outlineColor=<rgba 0 0 0 1>
     }
    }
    renderTransform {angle=0, scale=<1 1>, shear=<-14 0>, translation=<0 0>}
    shadowOffset=<0 0>
   }
   slot {class=BorderSlot, properties {verticalAlignment=VAlign_Center, padding=<ltrb 6 6 6 6>}}
  }
  Image_96 {
   widgetClassPath=&Image
   parent=VerticalBox_0
   widget=Image_96
   properties {
    visibility=SelfHitTestInvisible
    brush {
     bIsDynamicallyLoaded=false
     drawAs=Image
     imageType=NoImage
     mirroring=NoMirror
     resourceName=None
     resourceObject=/Game/Athena/UI/ReplayBrowser/M_Ui_Skewed_triangle
     tiling=NoTile
     imageSize=<16 12>
     margin=<ltrb 0 0 0 0>
     outlineSettings {
      bUseBrushTransparency=false
      roundingType=HalfHeightRadius
      width=0
      color {colorUseRule=UseColor_Specified, specifiedColor=<rgba 0 0 0 0>}
      cornerRadii=<0 0 0 0>
     }
     tintColor {colorUseRule=UseColor_Specified, specifiedColor=<rgba 1 1 1 1>}
     uVRegion {bIsValid=false, max=<0 0>, min=<0 0>}
    }
   }
   slot {class=VerticalBoxSlot, properties {horizontalAlignment=HAlign_Center, verticalAlignment=VAlign_Center}}
  }
  InputActionWidget {widgetClassPath=&CommonActionWidget, bInherited=true, widget=null}
 ]
}
