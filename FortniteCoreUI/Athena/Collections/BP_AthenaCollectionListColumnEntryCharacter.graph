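# Widget Blueprint for a column of character entries in an Athena collection list, derived from the native AthenaCollectionListGroupEntry class.
# Its only widget is the inherited root dynamic entry box, configured to stack a capped, spaced set of BP_AthenaCollectionListEntryCharacter widgets vertically, with matching designer previews.
# The event graph holds only unconnected stock events, so the Blueprint adds no logic beyond this configuration.
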WidgetBlueprint /FortniteCoreUI/Athena/Collections/BP_AthenaCollectionListColumnEntryCharacter
graphs [EventGraph {nodes [
   Event PreConstruct()
   Event Construct()
   Event Tick()
  ]}]
variables []
widgets {
 info {
  inheritedWidgetCount=1
  parentClass=/Script/FortniteUI.AthenaCollectionListGroupEntry
  rootWidgetClass=/Script/FortniteUI.FortDynamicEntryBox
  widgetCount=1
 }
 widgets [EntryBox_ListEntries {
   widgetClassPath=&FortDynamicEntryBox
   bInherited=true
   widget=EntryBox_ListEntries
   properties {
    entryBoxType=Vertical
    entryWidgetClass=/Game/Athena/Collections/BP_AthenaCollectionListEntryCharacter.BP_AthenaCollectionListEntryCharacter_C
    maxElementSize=176
    maximumEntries=4
    numDesignerPreviewEntries=4
    entrySpacing=<0 16>
   }
  }]
}
