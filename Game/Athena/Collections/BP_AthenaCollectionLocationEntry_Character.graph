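# Widget Blueprint for a fixed-height location row in the Athena character collection UI, built on the native AthenaCollectionLocationEntry class.
# Over a translucent blue background, a visibility switcher holds either the uppercase location name or a "???" not-placed label.
# The name overlay also contains a placement number and right-aligned material art, both collapsed by default.
# It has no graph logic, and all text uses a render shear for a slanted look.
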WidgetBlueprint /Game/Athena/Collections/BP_AthenaCollectionLocationEntry_Character
graphs [EventGraph {links [], nodes []}]
variables []
widgets {
 info {inheritedWidgetCount=3, parentClass=/Script/FortniteUI.AthenaCollectionLocationEntry, rootWidgetClass=&SizeBox, widgetCount=12}
 widgets [
  SizeBox_37 {widgetClassPath=&SizeBox, widget=SizeBox_37, properties {bOverride_HeightOverride=true, heightOverride=48}}
  Overlay_0 {widgetClassPath=&Overlay, parent=SizeBox_37, widget=Overlay_0, slot {class=SizeBoxSlot}}
  Image_BlueBg {
   widgetClassPath=&Image
   bInherited=true
   parent=Overlay_0
   widget=Image_BlueBg
   properties {
    visibility=SelfHitTestInvisible
    brush {
     bIsDynamicallyLoaded=false
     drawAs=Image
     imageType=NoImage
     mirroring=NoMirror
     resourceName=None
     resourceObject=null
     tiling=NoTile
     imageSize=<41.9809 33.1349>
     margin=<ltrb 0 0 0 0>
     outlineSettings {
      bUseBrushTransparency=false
      roundingType=HalfHeightRadius
      width=0
      color {colorUseRule=UseColor_Specified, specifiedColor=<rgba 0 0 0 0>}
      cornerRadii=<0 0 0 0>
     }
     tintColor {colorUseRule=UseColor_Specified, specifiedColor=<rgba 1 1 1 1>}
     uVRegion {bIsValid=false, max=<0 0>, min=<0 0>}
    }
    colorAndOpacity=<rgba 0 0.0908 1 0.2471>
   }
   slot {class=OverlaySlot, properties {horizontalAlignment=HAlign_Fill, verticalAlignment=VAlign_Fill}}
  }
  VisibilitySwitcher_PlacedNotPlaced {
   widgetClassPath=&CommonVisibilitySwitcher
   bInherited=true
   bIsVariable=true
   parent=Overlay_0
   widget=VisibilitySwitcher_PlacedNotPlaced
   slot {class=OverlaySlot, properties {horizontalAlignment=HAlign_Fill, verticalAlignment=VAlign_Fill, padding=<ltrb 6 0 0 0>}}
  }
  Overlay_1 {widgetClassPath=&Overlay, parent=VisibilitySwitcher_PlacedNotPlaced, widget=Overlay_1, slot {class=CommonVisibilitySwitcherSlot}}
  HorizontalBox_Placement {
   widgetClassPath=&HorizontalBox
   parent=Overlay_1
   widget=HorizontalBox_Placement
   properties {visibility=Collapsed}
   slot {class=OverlaySlot, properties {horizontalAlignment=HAlign_Fill, verticalAlignment=VAlign_Center}}
  }
  CommonTextBlock_76 {
   widgetClassPath=&CommonTextBlock
   parent=HorizontalBox_Placement
   widget=CommonTextBlock_76
   properties {
    text="##"
    colorAndOpacity {colorUseRule=UseColor_Specified, specifiedColor=<rgba 0 0.0908 1 0.498>}
    font {
     bForceMonospaced=false
     bMaterialIsStencil=false
     fontMaterial=null
     fontObject=&BurbankSmall
     letterSpacing=0
     monospacedWidth=1
     size=18
     skewAmount=0
     typefaceFontName=Black
     outlineSettings {
      bApplyOutlineToDropShadows=false
      bMiteredCorners=false
      bSeparateFillAlpha=false
      outlineMaterial=null
      outlineSize=0
      outlineColor=<rgba 0 0 0 1>
     }
    }
    margin=<ltrb 0 -16 0 0>
    renderTransform {angle=0, scale=<1 1>, shear=<-12 0>, translation=<0 0>}
    shadowOffset=<0 0>
   }
   slot {class=HorizontalBoxSlot, properties {verticalAlignment=VAlign_Center}}
  }
  Text_Placement {
   widgetClassPath=&CommonTextBlock
   parent=HorizontalBox_Placement
   widget=Text_Placement
   properties {
    text=1
    colorAndOpacity {colorUseRule=UseColor_Specified, specifiedColor=<rgba 0 0.0908 1 0.498>}
    font {
     bForceMonospaced=false
     bMaterialIsStencil=false
     fontMaterial=null
     fontObject=&BurbankSmall
     letterSpacing=0
     monospacedWidth=1
     size=24
     skewAmount=0
     typefaceFontName=Black
     outlineSettings {
      bApplyOutlineToDropShadows=false
      bMiteredCorners=false
      bSeparateFillAlpha=false
      outlineMaterial=null
      outlineSize=0
      outlineColor=<rgba 0 0 0 1>
     }
    }
    renderTransform {angle=0, scale=<1 1>, shear=<-12 0>, translation=<0 0>}
    shadowOffset=<0 0>
   }
   slot {class=HorizontalBoxSlot}
  }
  Text_DisplayName {
   widgetClassPath=&CommonTextBlock
   bInherited=true
   parent=Overlay_1
   widget=Text_DisplayName
   properties {
    mobileFontSizeMultiplier=1.25
    scrollStyle=&TextScrollStyle-Base_C
    text=Location Name
    textTransformPolicy=ToUpper
    font {
     bForceMonospaced=false
     bMaterialIsStencil=false
     fontMaterial=null
     fontObject=&BurbankSmall
     letterSpacing=0
     monospacedWidth=1
     size=16
     skewAmount=0
     typefaceFontName=Black
     outlineSettings {
      bApplyOutlineToDropShadows=false
      bMiteredCorners=false
      bSeparateFillAlpha=false
      outlineMaterial=null
      outlineSize=0
      outlineColor=<rgba 0 0 0 1>
     }
    }
    renderTransform {angle=0, scale=<1 1>, shear=<-12 0>, translation=<0 0>}
    shadowOffset=<0 0>
   }
   slot {class=OverlaySlot, properties {horizontalAlignment=HAlign_Fill, verticalAlignment=VAlign_Center, padding=<ltrb 12 0 128 0>}}
  }
  ScaleBox_0 {
   widgetClassPath=&ScaleBox
   parent=Overlay_1
   widget=ScaleBox_0
   properties {stretch=ScaleToFitY, visibility=Collapsed}
   slot {class=OverlaySlot, properties {horizontalAlignment=HAlign_Right, verticalAlignment=VAlign_Fill}}
  }
  Image_81 {
   widgetClassPath=&Image
   parent=ScaleBox_0
   widget=Image_81
   properties {brush {
     bIsDynamicallyLoaded=false
     drawAs=Image
     imageType=NoImage
     mirroring=NoMirror
     resourceName=None
     resourceObject=/Game/Athena/Collections/Art/Character/T_Ui_Collection_Location_Mat_Inst
     tiling=NoTile
     imageSize=<256 64>
     margin=<ltrb 0 0 0 0>
     outlineSettings {
      bUseBrushTransparency=false
      roundingType=HalfHeightRadius
      width=0
      color {colorUseRule=UseColor_Specified, specifiedColor=<rgba 0 0 0 0>}
      cornerRadii=<0 0 0 0>
     }
     tintColor {colorUseRule=UseColor_Specified, specifiedColor=<rgba 1 1 1 0.75>}
     uVRegion {bIsValid=false, max=<0 0>, min=<0 0>}
    }}
   slot {class=ScaleBoxSlot}
  }
  Text_NotPlaced {
   widgetClassPath=&CommonTextBlock
   parent=VisibilitySwitcher_PlacedNotPlaced
   widget=Text_NotPlaced
   properties {
    text=???
    textTransformPolicy=ToUpper
    colorAndOpacity {colorUseRule=UseColor_Specified, specifiedColor=<rgba 0.004 0.1683 1 1>}
    font {
     bForceMonospaced=false
     bMaterialIsStencil=false
     fontMaterial=null
     fontObject=&BurbankSmall
     letterSpacing=0
     monospacedWidth=1
     size=16
     skewAmount=0
     typefaceFontName=Black
     outlineSettings {
      bApplyOutlineToDropShadows=false
      bMiteredCorners=false
      bSeparateFillAlpha=false
      outlineMaterial=null
      outlineSize=0
      outlineColor=<rgba 0 0 0 1>
     }
    }
    renderTransform {angle=0, scale=<1 1>, shear=<-12 0>, translation=<0 0>}
    shadowOffset=<0 0>
   }
   slot {class=CommonVisibilitySwitcherSlot, properties {verticalAlignment=VAlign_Center, padding=<ltrb 12 0 12 0>}}
  }
 ]
}
