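# Info overlay for the Athena character collection screen, showing image, name, number and descriptions on top of the native AthenaCollectionScreenInfoOverlayCharacter class.
# Setting the collected state replays the ReloadInfo intro, swaps info for a question mark when Unknown, and shows descriptions from NewlyCollected onward.
# It also resets the default-collapsed location switcher to its first, three-entry list.
# Location found and total counts are stored but unused in this graph.
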WidgetBlueprint /FortniteCoreUI/Athena/Collections/BP_AthenaCollectionScreenInfoOverlayCharacter
animations [ReloadInfo {
  binding_count=5
  length_seconds=0.25
  object=&self:ReloadInfo
  sequence {bindings [
    Image_ItemImage {
     id=#1
     possessed_class=Image
     tracks [{
       type=MovieSceneWidgetMaterialTrack
       sections [{type=MovieSceneParameterSection, channels [Intro {type=MovieSceneScriptingFloatChannel, keys [[0,0],[5,1]]}]}]
      }]
    }
    RichText_AdditionalDescription {
     id=#2
     possessed_class=CommonRichTextBlock
     tracks [{
       type=MovieSceneFloatTrack
       display_name=Render Opacity
       sections [{type=MovieSceneFloatSection, channels [None {type=MovieSceneScriptingFloatChannel, keys [[1,0],[5,1]]}]}]
      }]
    }
    Text_ItemName {
     id=#3
     possessed_class=CommonTextBlock
     tracks [
      {
       type=MovieScene2DTransformTrack
       sections [{
         type=MovieScene2DTransformSection
         channels [
          Translation.X {type=MovieSceneScriptingFloatChannel}
          Translation.Y {type=MovieSceneScriptingFloatChannel, keys [[0,32],[3,0]]}
         ]
        }]
      }
      {
       type=MovieSceneFloatTrack
       display_name=Render Opacity
       sections [{type=MovieSceneFloatSection, channels [None {type=MovieSceneScriptingFloatChannel, keys [[0,0],[3,1]]}]}]
      }
     ]
    }
    RichText_Description {
     id=#4
     possessed_class=CommonRichTextBlock
     tracks [
      {
       type=MovieScene2DTransformTrack
       sections [{
         type=MovieScene2DTransformSection
         channels [
          Translation.X {type=MovieSceneScriptingFloatChannel}
          Translation.Y {type=MovieSceneScriptingFloatChannel, keys [[1,24],[4,0]]}
         ]
        }]
      }
      {
       type=MovieSceneFloatTrack
       display_name=Render Opacity
       sections [{type=MovieSceneFloatSection, channels [None {type=MovieSceneScriptingFloatChannel, keys [[1,0],[4,1]]}]}]
      }
     ]
    }
    Text_ItemNumber {
     id=#5
     possessed_class=CommonTextBlock
     tracks [
      {
       type=MovieScene2DTransformTrack
       sections [{
         type=MovieScene2DTransformSection
         channels [
          Translation.X {type=MovieSceneScriptingFloatChannel}
          Translation.Y {type=MovieSceneScriptingFloatChannel, keys [[0,24],[5,0]]}
          Angle {type=MovieSceneScriptingFloatChannel}
          Scale.X {type=MovieSceneScriptingFloatChannel}
          Scale.Y {type=MovieSceneScriptingFloatChannel}
          Shear.X {type=MovieSceneScriptingFloatChannel}
          Shear.Y {type=MovieSceneScriptingFloatChannel}
         ]
        }]
      }
      {
       type=MovieSceneFloatTrack
       display_name=Render Opacity
       sections [{type=MovieSceneFloatSection, channels [None {type=MovieSceneScriptingFloatChannel, keys [[0,0],[5,1]]}]}]
      }
     ]
    }
   ]}
 }]
graphs [EventGraph {nodes [
   n1 = Event On Collected State Set!() -> (OutputDelegate, then, CollectedState:EFortCollectedState Enum=Unknown) {is=title K2Node_Event}
   > PlayAnimationForward(InAnimation=Get ReloadInfo() -> (ReloadInfo:Widget Animation Object Reference))
   > n2 = Set State(State:EFortCollectedState Enum=@n1.CollectedState) -> (Output_Get:EFortCollectedState Enum=Unknown)
   n3 = "Equal (Enum)"(A:EFortCollectedState Enum=@n2, B:EFortCollectedState Enum=Unknown)
   > SetVisibility(
    self=Get VerticalBox_AllInfos() -> (VerticalBox_AllInfos:Vertical Box Object Reference)
    InVisibility=Select(Option 0=HitTestInvisible, Option 1=Collapsed, Index=@n3)
   )
   > SetVisibility(
    self=Get QuestionMark() -> (QuestionMark:Common Text Object Reference)
    InVisibility=Select(Option 0=Collapsed, Option 1=SelfHitTestInvisible, Index=@n3)
   )
   > SetVisibility(
    self=Get SizeBox_Descriptions() -> (SizeBox_Descriptions:Size Box Object Reference)
    InVisibility=Select!(
     Unknown:ESlateVisibility Enum=Collapsed
     New:ESlateVisibility Enum=Collapsed
     Known:ESlateVisibility Enum=Collapsed
     NewlyCollected:ESlateVisibility Enum=HitTestInvisible
     Collected:ESlateVisibility Enum=HitTestInvisible
     NewBest:ESlateVisibility Enum=HitTestInvisible
     NewRecord:ESlateVisibility Enum=HitTestInvisible
     NewLocation:ESlateVisibility Enum=Visible
     NewlyCompleted:ESlateVisibility Enum=Visible
     Complete:ESlateVisibility Enum=Visible
     Index:EFortCollectedState Enum=@n2
    ) -> (ReturnValue)
   )
   > SetActiveWidgetIndex(self=Get Switcher_Locations() -> (Switcher_Locations:Widget Switcher Object Reference), Index=0)
   n4 = Event On Locations Found Set!() -> (OutputDelegate, then, NumFound:int=0, NumAvailable:int=0, bAllFound:bool) {is=title K2Node_Event}
   > Set Current location(Current location:int=@n4.NumFound) -> (Output_Get:int=0)
   > Set Total location(Total location:int=@n4.NumAvailable) -> (Output_Get:int=0)
   Event Construct()
  ]}]
variables [
 Current location {type=integer, default=1}
 LocationRGBA2_Base {type=LinearColor, default=<rgba 0 0.0203 0.2747 1>}
 LocationRGBA2_Hover {type=LinearColor, default=<rgba 0 0.0423 0.5395 1>}
 LocationTrackerColor {type=LinearColor, default=<rgba 0.2159 0.6584 1 1>}
 LocationTrackerColor_Complete {type=LinearColor, default=<rgba 1 0.8228 0 1>}
 State {type=EFortCollectedState, default=Unknown}
 Total location {type=integer, default=3}
]
widgets {
 info {
  inheritedWidgetCount=12
  parentClass=/Script/FortniteUI.AthenaCollectionScreenInfoOverlayCharacter
  rootWidgetClass=&SizeBox
  widgetCount=34
 }
 widgets [
  SizeBox_5 {widgetClassPath=&SizeBox, widget=SizeBox_5, properties {bOverride_WidthOverride=true, widthOverride=550}}
  Overlay_0 {widgetClassPath=&Overlay, parent=SizeBox_5, widget=Overlay_0, slot {class=SizeBoxSlot}}
  Image_118 {
   widgetClassPath=&Image
   bIsVariable=true
   parent=Overlay_0
   widget=Image_118
   properties {
    visibility=HitTestInvisible
    brush {
     bIsDynamicallyLoaded=false
     drawAs=Box
     imageType=NoImage
     mirroring=NoMirror
     resourceName=None
     resourceObject=/Game/UI/Style/Materials/BasicLinearUIGradient_CommonBorder_Inst
     tiling=NoTile
     imageSize=<90 90>
     margin=<ltrb 0 0.5 0 0.5>
     outlineSettings {
      bUseBrushTransparency=false
      roundingType=HalfHeightRadius
      width=0
      color {colorUseRule=UseColor_Specified, specifiedColor=<rgba 0 0 0 0>}
      cornerRadii=<0 0 0 0>
     }
     tintColor {colorUseRule=UseColor_Specified, specifiedColor=<rgba 0 0 0 0.25>}
     uVRegion {bIsValid=false, max=<0 0>, min=<0 0>}
    }
    renderTransform {angle=0, scale=<1 1>, shear=<2 0>, translation=<0 0>}
   }
   slot {class=OverlaySlot, properties {verticalAlignment=VAlign_Fill, padding=<ltrb -150 -80 0 -100>}}
  }
  Image_64 {
   widgetClassPath=&Image
   bIsVariable=true
   parent=Overlay_0
   widget=Image_64
   properties {
    brush {
     bIsDynamicallyLoaded=false
     drawAs=Image
     imageType=NoImage
     mirroring=NoMirror
     resourceName=None
     resourceObject=/Game/UI/Frontend/StoreMain/Art/M_UI_RMT_SidePanel_Dark
     tiling=NoTile
     imageSize=<32 32>
     margin=<ltrb 0 0 0 0>
     outlineSettings {
      bUseBrushTransparency=false
      roundingType=HalfHeightRadius
      width=0
      color {colorUseRule=UseColor_Specified, specifiedColor=<rgba 0 0 0 0>}
      cornerRadii=<0 0 0 0>
     }
     tintColor {colorUseRule=UseColor_Specified, specifiedColor=<rgba 1 1 1 1>}
     uVRegion {bIsValid=false, max=<0 0>, min=<0 0>}
    }
    renderTransform {angle=0, scale=<-1 1>, shear=<0 0>, translation=<0 0>}
   }
   slot {class=OverlaySlot, properties {horizontalAlignment=HAlign_Fill, verticalAlignment=VAlign_Fill, padding=<ltrb -95 -82 -200 -100>}}
  }
  VerticalBox_22 {
   widgetClassPath=&VerticalBox
   parent=Overlay_0
   widget=VerticalBox_22
   slot {class=OverlaySlot, properties {horizontalAlignment=HAlign_Fill, verticalAlignment=VAlign_Fill}}
  }
  Overlay_25 {
   widgetClassPath=&Overlay
   parent=VerticalBox_22
   widget=Overlay_25
   slot {class=VerticalBoxSlot, properties {size {sizeRule=Fill, value=1}}}
  }
  Image_RadialBloom {
   widgetClassPath=&Image
   bIsVariable=true
   parent=Overlay_25
   widget=Image_RadialBloom
   properties {
    brush {
     bIsDynamicallyLoaded=false
     drawAs=Image
     imageType=NoImage
     mirroring=NoMirror
     resourceName=None
     resourceObject=/Game/Athena/Collections/Art/Character/M_Ui_RadialGradient
     tiling=NoTile
     imageSize=<512 512>
     margin=<ltrb 0 0 0 0>
     outlineSettings {
      bUseBrushTransparency=false
      roundingType=HalfHeightRadius
      width=0
      color {colorUseRule=UseColor_Specified, specifiedColor=<rgba 0 0 0 0>}
      cornerRadii=<0 0 0 0>
     }
     tintColor {colorUseRule=UseColor_Specified, specifiedColor=<rgba 0 0.0908 1 1>}
     uVRegion {bIsValid=false, max=<0 0>, min=<0 0>}
    }
    colorAndOpacity=<rgba 1 1 1 0.5>
    renderTransform {angle=0, scale=<1.2 1.2>, shear=<0 0>, translation=<0 0>}
   }
   slot {class=OverlaySlot, properties {horizontalAlignment=HAlign_Center, padding=<ltrb 0 -30 0 0>}}
  }
  CommonVisibilityWidget_0 {
   widgetClassPath=&CommonVisibilityWidgetLegacy
   parent=Overlay_25
   widget=CommonVisibilityWidget_0
   properties {bShowForAndroid=false, bShowForIOS=false, bShowForTouch=false}
   slot {class=OverlaySlot, properties {horizontalAlignment=HAlign_Right}}
  }
  Text_ItemNumber {
   widgetClassPath=&CommonTextBlock
   bInherited=true
   parent=CommonVisibilityWidget_0
   widget=Text_ItemNumber
   properties {
    lineHeightPercentage=0.8095
    text=01
    colorAndOpacity {colorUseRule=UseColor_Specified, specifiedColor=<rgba 0 0.0908 1 0.25>}
    font {
     bForceMonospaced=false
     bMaterialIsStencil=false
     fontMaterial=null
     fontObject=&Burbank
     letterSpacing=0
     monospacedWidth=1
     size=100
     skewAmount=0
     typefaceFontName=BurbankRegular-Black
     outlineSettings {
      bApplyOutlineToDropShadows=false
      bMiteredCorners=false
      bSeparateFillAlpha=false
      outlineMaterial=null
      outlineSize=0
      outlineColor=<rgba 0 0 0 1>
     }
    }
    margin=<ltrb 0 -32 0 0>
    renderTransform {angle=0, scale=<1 1>, shear=<-12 0>, translation=<0 0>}
    shadowOffset=<0 0>
   }
   slot {class=BorderSlot}
  }
  Image_ItemImage {
   widgetClassPath=&Image
   bInherited=true
   bIsVariable=true
   parent=Overlay_25
   widget=Image_ItemImage
   properties {brush {
     bIsDynamicallyLoaded=false
     drawAs=Image
     imageType=NoImage
     mirroring=NoMirror
     resourceName=None
     resourceObject=null
     tiling=NoTile
     imageSize=<1000 1000>
     margin=<ltrb 0 0 0 0>
     outlineSettings {
      bUseBrushTransparency=false
      roundingType=HalfHeightRadius
      width=0
      color {colorUseRule=UseColor_Specified, specifiedColor=<rgba 0 0 0 0>}
      cornerRadii=<0 0 0 0>
     }
     tintColor {colorUseRule=UseColor_Specified, specifiedColor=<rgba 1 1 1 1>}
     uVRegion {bIsValid=false, max=<0 0>, min=<0 0>}
    }}
   slot {class=OverlaySlot, properties {horizontalAlignment=HAlign_Center, padding=<ltrb -300 0 -300 -512>}}
  }
  QuestionMark {
   widgetClassPath=&CommonTextBlock
   bIsVariable=true
   parent=Overlay_25
   widget=QuestionMark
   properties {
    text=?
    visibility=Collapsed
    colorAndOpacity {colorUseRule=UseColor_Specified, specifiedColor=<rgba 0 0.0908 1 1>}
    font {
     bForceMonospaced=false
     bMaterialIsStencil=false
     fontMaterial=null
     fontObject=&BurbankSmall
     letterSpacing=0
     monospacedWidth=1
     size=150
     skewAmount=0
     typefaceFontName=Bold
     outlineSettings {
      bApplyOutlineToDropShadows=false
      bMiteredCorners=false
      bSeparateFillAlpha=false
      outlineMaterial=null
      outlineSize=0
      outlineColor=<rgba 0 0 0 1>
     }
    }
    renderTransform {angle=0, scale=<1 1>, shear=<-12 0>, translation=<-10 196>}
    shadowOffset=<0 0>
   }
   slot {class=OverlaySlot, properties {horizontalAlignment=HAlign_Center}}
  }
  VerticalBox_AllInfos {
   widgetClassPath=&VerticalBox
   bIsVariable=true
   parent=Overlay_25
   widget=VerticalBox_AllInfos
   slot {class=OverlaySlot, properties {horizontalAlignment=HAlign_Fill, verticalAlignment=VAlign_Bottom}}
  }
  Text_ItemType {
   widgetClassPath=&CommonTextBlock
   bInherited=true
   parent=VerticalBox_AllInfos
   widget=Text_ItemType
   properties {
    justification=Center
    text="{SPECIES}"
    textTransformPolicy=ToUpper
    visibility=Collapsed
    colorAndOpacity {colorUseRule=UseColor_Specified, specifiedColor=<rgba 0.0331 0.855 1 1>}
    font {
     bForceMonospaced=false
     bMaterialIsStencil=false
     fontMaterial=null
     fontObject=&BurbankSmall
     letterSpacing=0
     monospacedWidth=1
     size=20
     skewAmount=0
     typefaceFontName=Black
     outlineSettings {
      bApplyOutlineToDropShadows=false
      bMiteredCorners=false
      bSeparateFillAlpha=false
      outlineMaterial=null
      outlineSize=0
      outlineColor=<rgba 0 0 0 1>
     }
    }
    renderTransform {angle=0, scale=<1 1>, shear=<-12 0>, translation=<0 0>}
    shadowOffset=<0 0>
   }
   slot {class=VerticalBoxSlot}
  }
  Text_ItemName {
   widgetClassPath=&CommonTextBlock
   bInherited=true
   parent=VerticalBox_AllInfos
   widget=Text_ItemName
   properties {
    autoWrapText=true
    style=/FortUILibrary/Styles/Typography/_TextStyle_SubH2._TextStyle_SubH2_C
    text="{CHARACTER NAME}"
    textTransformPolicy=ToUpper
    font {
     bForceMonospaced=false
     bMaterialIsStencil=false
     fontMaterial=null
     fontObject=&Burbank
     letterSpacing=0
     monospacedWidth=1
     size=33
     skewAmount=0
     typefaceFontName=BurbankRegular-Black
     outlineSettings {
      bApplyOutlineToDropShadows=false
      bMiteredCorners=false
      bSeparateFillAlpha=false
      outlineMaterial=null
      outlineSize=0
      outlineColor=<rgba 0 0 0 1>
     }
    }
    shadowOffset=<0 0>
    strikeBrush {
     bIsDynamicallyLoaded=false
     drawAs=NoDrawType
     imageType=NoImage
     mirroring=NoMirror
     resourceName=None
     resourceObject=null
     tiling=NoTile
     imageSize=<32 32>
     margin=<ltrb 0 0 0 0>
     outlineSettings {
      bUseBrushTransparency=false
      roundingType=HalfHeightRadius
      width=0
      color {colorUseRule=UseColor_Specified, specifiedColor=<rgba 0 0 0 0>}
      cornerRadii=<0 0 0 0>
     }
     tintColor {colorUseRule=UseColor_Specified, specifiedColor=<rgba 1 1 1 1>}
     uVRegion {bIsValid=false, max=<0 0>, min=<0 0>}
    }
   }
   slot {class=VerticalBoxSlot}
  }
  VerticalBox_25 {
   widgetClassPath=&VerticalBox
   parent=VerticalBox_AllInfos
   widget=VerticalBox_25
   slot {class=VerticalBoxSlot, properties {size {sizeRule=Fill, value=1}}}
  }
  SizeBox_Descriptions {
   widgetClassPath=&SizeBox
   bIsVariable=true
   parent=VerticalBox_25
   widget=SizeBox_Descriptions
   properties {bOverride_MaxDesiredHeight=true, maxDesiredHeight=400}
   slot {class=VerticalBoxSlot}
  }
  ScrollBox_0 {
   widgetClassPath=&ScrollBox
   parent=SizeBox_Descriptions
   widget=ScrollBox_0
   properties {visibility=SelfHitTestInvisible}
   slot {class=SizeBoxSlot}
  }
  RichText_Description {
   widgetClassPath=&CommonRichTextBlock
   bInherited=true
   parent=ScrollBox_0
   widget=RichText_Description
   properties {
    autoWrapText=true
    bOverrideDefaultStyle=true
    defaultTextStyleOverrideClass=/FortUILibrary/Styles/Typography/_TextStyle_SubH4_Cyan._TextStyle_SubH4_Cyan_C
    mobileTextBlockScale=1.3
    text="<text color=\"fff\" fontface=\"black\" case=\"upper\">{PH}</> This is the <text color=\"fff\" fontface=\"black\" case=\"upper\">character description</>. No shirt. No surrender."
    textTransformPolicy=ToUpper
    defaultTextStyleOverride {
     overflowPolicy=Clip
     transformPolicy=None
     colorAndOpacity {colorUseRule=UseColor_Specified, specifiedColor=<rgba 0 1 0.9647 1>}
     font {
      bForceMonospaced=false
      bMaterialIsStencil=false
      fontMaterial=null
      fontObject=&BurbankSmall
      letterSpacing=0
      monospacedWidth=1
      size=17
      skewAmount=0
      typefaceFontName=Black
      outlineSettings {
       bApplyOutlineToDropShadows=false
       bMiteredCorners=false
       bSeparateFillAlpha=false
       outlineMaterial=null
       outlineSize=0
       outlineColor=<rgba 0 0 0 1>
      }
     }
     highlightColor {colorUseRule=UseColor_Specified, specifiedColor=<rgba 0 0 0 1>}
     highlightShape {
      bIsDynamicallyLoaded=false
      drawAs=Image
      imageType=NoImage
      mirroring=NoMirror
      resourceName=None
      resourceObject=null
      tiling=NoTile
      imageSize=<32 32>
      margin=<ltrb 0 0 0 0>
      outlineSettings {
       bUseBrushTransparency=false
       roundingType=HalfHeightRadius
       width=0
       color {colorUseRule=UseColor_Specified, specifiedColor=<rgba 0 0 0 0>}
       cornerRadii=<0 0 0 0>
      }
      tintColor {colorUseRule=UseColor_Specified, specifiedColor=<rgba 1 1 1 1>}
      uVRegion {bIsValid=false, max=<0 0>, min=<0 0>}
     }
     selectedBackgroundColor {colorUseRule=UseColor_Foreground, specifiedColor=<rgba 1 0 1 1>}
     shadowColorAndOpacity=<rgba 0 0 0 1>
     shadowOffset=<0 0>
     strikeBrush {
      bIsDynamicallyLoaded=false
      drawAs=Image
      imageType=NoImage
      mirroring=NoMirror
      resourceName=None
      resourceObject=null
      tiling=NoTile
      imageSize=<32 32>
      margin=<ltrb 0 0 0 0>
      outlineSettings {
       bUseBrushTransparency=false
       roundingType=HalfHeightRadius
       width=0
       color {colorUseRule=UseColor_Specified, specifiedColor=<rgba 0 0 0 0>}
       cornerRadii=<0 0 0 0>
      }
      tintColor {colorUseRule=UseColor_Specified, specifiedColor=<rgba 1 1 1 1>}
      uVRegion {bIsValid=false, max=<0 0>, min=<0 0>}
     }
     underlineBrush {
      bIsDynamicallyLoaded=false
      drawAs=Image
      imageType=NoImage
      mirroring=NoMirror
      resourceName=None
      resourceObject=null
      tiling=NoTile
      imageSize=<32 32>
      margin=<ltrb 0 0 0 0>
      outlineSettings {
       bUseBrushTransparency=false
       roundingType=HalfHeightRadius
       width=0
       color {colorUseRule=UseColor_Specified, specifiedColor=<rgba 0 0 0 0>}
       cornerRadii=<0 0 0 0>
      }
      tintColor {colorUseRule=UseColor_Specified, specifiedColor=<rgba 1 1 1 1>}
      uVRegion {bIsValid=false, max=<0 0>, min=<0 0>}
     }
    }
   }
   slot {class=ScrollBoxSlot, properties {padding=<ltrb 0 0 0 6>}}
  }
  RichText_AdditionalDescription {
   widgetClassPath=&CommonRichTextBlock
   bInherited=true
   parent=ScrollBox_0
   widget=RichText_AdditionalDescription
   properties {
    autoWrapText=true
    bOverrideDefaultStyle=true
    defaultTextStyleOverrideClass=/FortUILibrary/Styles/Typography/_TextStyle_Body2_LightBlue._TextStyle_Body2_LightBlue_C
    mobileTextBlockScale=1.3
    text="<text color=\"fff\" fontface=\"black\" case=\"upper\">{PH}</> This is the <text color=\"fff\" fontface=\"black\" case=\"upper\">Character infos</>. Handsome and suave, Bandolier is an old-school mercenary whose pecs pack as much punch as his shotgun. Show him your moves, and you'll be swapping workout tips in no time."
    defaultTextStyleOverride {
     overflowPolicy=Clip
     transformPolicy=None
     colorAndOpacity {colorUseRule=UseColor_Specified, specifiedColor=<rgba 0.2122 0.6584 1 1>}
     font {
      bForceMonospaced=false
      bMaterialIsStencil=false
      fontMaterial=null
      fontObject=&BurbankSmall
      letterSpacing=0
      monospacedWidth=1
      size=14
      skewAmount=0
      typefaceFontName=Bold
      outlineSettings {
       bApplyOutlineToDropShadows=false
       bMiteredCorners=false
       bSeparateFillAlpha=false
       outlineMaterial=null
       outlineSize=0
       outlineColor=<rgba 0 0 0 1>
      }
     }
     highlightColor {colorUseRule=UseColor_Specified, specifiedColor=<rgba 0 0 0 1>}
     highlightShape {
      bIsDynamicallyLoaded=false
      drawAs=Image
      imageType=NoImage
      mirroring=NoMirror
      resourceName=None
      resourceObject=null
      tiling=NoTile
      imageSize=<32 32>
      margin=<ltrb 0 0 0 0>
      outlineSettings {
       bUseBrushTransparency=false
       roundingType=HalfHeightRadius
       width=0
       color {colorUseRule=UseColor_Specified, specifiedColor=<rgba 0 0 0 0>}
       cornerRadii=<0 0 0 0>
      }
      tintColor {colorUseRule=UseColor_Specified, specifiedColor=<rgba 1 1 1 1>}
      uVRegion {bIsValid=false, max=<0 0>, min=<0 0>}
     }
     selectedBackgroundColor {colorUseRule=UseColor_Foreground, specifiedColor=<rgba 1 0 1 1>}
     shadowColorAndOpacity=<rgba 0 0 0 1>
     shadowOffset=<0 0>
     strikeBrush {
      bIsDynamicallyLoaded=false
      drawAs=Image
      imageType=NoImage
      mirroring=NoMirror
      resourceName=None
      resourceObject=null
      tiling=NoTile
      imageSize=<32 32>
      margin=<ltrb 0 0 0 0>
      outlineSettings {
       bUseBrushTransparency=false
       roundingType=HalfHeightRadius
       width=0
       color {colorUseRule=UseColor_Specified, specifiedColor=<rgba 0 0 0 0>}
       cornerRadii=<0 0 0 0>
      }
      tintColor {colorUseRule=UseColor_Specified, specifiedColor=<rgba 1 1 1 1>}
      uVRegion {bIsValid=false, max=<0 0>, min=<0 0>}
     }
     underlineBrush {
      bIsDynamicallyLoaded=false
      drawAs=Image
      imageType=NoImage
      mirroring=NoMirror
      resourceName=None
      resourceObject=null
      tiling=NoTile
      imageSize=<32 32>
      margin=<ltrb 0 0 0 0>
      outlineSettings {
       bUseBrushTransparency=false
       roundingType=HalfHeightRadius
       width=0
       color {colorUseRule=UseColor_Specified, specifiedColor=<rgba 0 0 0 0>}
       cornerRadii=<0 0 0 0>
      }
      tintColor {colorUseRule=UseColor_Specified, specifiedColor=<rgba 1 1 1 1>}
      uVRegion {bIsValid=false, max=<0 0>, min=<0 0>}
     }
    }
   }
   slot {class=ScrollBoxSlot, properties {padding=<ltrb 4 4 4 4>}}
  }
  RichText_AdditionalInfo {
   widgetClassPath=&CommonRichTextBlock
   bInherited=true
   bIsVariable=true
   parent=ScrollBox_0
   widget=RichText_AdditionalInfo
   properties {
    autoWrapText=true
    bAutoCollapseWithEmptyText=true
    bOverrideDefaultStyle=true
    justification=Center
    mobileTextBlockScale=1.3
    text=" <text color=\"fff\" fontface=\"black\" case=\"upper\">{PH}</> This is the <text color=\"fff\" fontface=\"black\" case=\"upper\">Character optional infos</>."
    defaultTextStyleOverride {
     overflowPolicy=Clip
     transformPolicy=None
     colorAndOpacity {colorUseRule=UseColor_Specified, specifiedColor=<rgba 0.2122 0.6584 1 1>}
     font {
      bForceMonospaced=false
      bMaterialIsStencil=false
      fontMaterial=null
      fontObject=&BurbankSmall
      letterSpacing=0
      monospacedWidth=1
      size=14
      skewAmount=0
      typefaceFontName=Bold
      outlineSettings {
       bApplyOutlineToDropShadows=false
       bMiteredCorners=false
       bSeparateFillAlpha=false
       outlineMaterial=null
       outlineSize=0
       outlineColor=<rgba 0 0 0 1>
      }
     }
     highlightColor {colorUseRule=UseColor_Specified, specifiedColor=<rgba 0 0 0 1>}
     highlightShape {
      bIsDynamicallyLoaded=false
      drawAs=Image
      imageType=NoImage
      mirroring=NoMirror
      resourceName=None
      resourceObject=null
      tiling=NoTile
      imageSize=<32 32>
      margin=<ltrb 0 0 0 0>
      outlineSettings {
       bUseBrushTransparency=false
       roundingType=HalfHeightRadius
       width=0
       color {colorUseRule=UseColor_Specified, specifiedColor=<rgba 0 0 0 0>}
       cornerRadii=<0 0 0 0>
      }
      tintColor {colorUseRule=UseColor_Specified, specifiedColor=<rgba 1 1 1 1>}
      uVRegion {bIsValid=false, max=<0 0>, min=<0 0>}
     }
     selectedBackgroundColor {colorUseRule=UseColor_Foreground, specifiedColor=<rgba 1 0 1 1>}
     shadowColorAndOpacity=<rgba 0 0 0 1>
     shadowOffset=<0 0>
     strikeBrush {
      bIsDynamicallyLoaded=false
      drawAs=Image
      imageType=NoImage
      mirroring=NoMirror
      resourceName=None
      resourceObject=null
      tiling=NoTile
      imageSize=<32 32>
      margin=<ltrb 0 0 0 0>
      outlineSettings {
       bUseBrushTransparency=false
       roundingType=HalfHeightRadius
       width=0
       color {colorUseRule=UseColor_Specified, specifiedColor=<rgba 0 0 0 0>}
       cornerRadii=<0 0 0 0>
      }
      tintColor {colorUseRule=UseColor_Specified, specifiedColor=<rgba 1 1 1 1>}
      uVRegion {bIsValid=false, max=<0 0>, min=<0 0>}
     }
     underlineBrush {
      bIsDynamicallyLoaded=false
      drawAs=Image
      imageType=NoImage
      mirroring=NoMirror
      resourceName=None
      resourceObject=null
      tiling=NoTile
      imageSize=<32 32>
      margin=<ltrb 0 0 0 0>
      outlineSettings {
       bUseBrushTransparency=false
       roundingType=HalfHeightRadius
       width=0
       color {colorUseRule=UseColor_Specified, specifiedColor=<rgba 0 0 0 0>}
       cornerRadii=<0 0 0 0>
      }
      tintColor {colorUseRule=UseColor_Specified, specifiedColor=<rgba 1 1 1 1>}
      uVRegion {bIsValid=false, max=<0 0>, min=<0 0>}
     }
    }
    renderTransform {angle=0, scale=<1 1>, shear=<-12 0>, translation=<0 0>}
   }
   slot {class=ScrollBoxSlot, properties {padding=<ltrb 0 4 0 4>}}
  }
  Panel_FriendScores {
   widgetClassPath=&VerticalBox
   bInherited=true
   parent=VerticalBox_25
   widget=Panel_FriendScores
   slot {class=VerticalBoxSlot, properties {padding=<ltrb 0 12 0 0>}}
  }
  Switcher_Locations {
   widgetClassPath=&WidgetSwitcher
   bIsVariable=true
   parent=Panel_FriendScores
   widget=Switcher_Locations
   properties {visibility=Collapsed}
   slot {class=VerticalBoxSlot, properties {verticalAlignment=VAlign_Bottom, padding=<ltrb 0 6 0 0>, size {sizeRule=Fill, value=1}}}
  }
  SizeBox_1 {
   widgetClassPath=&SizeBox
   parent=Switcher_Locations
   widget=SizeBox_1
   properties {bOverride_HeightOverride=true, heightOverride=156}
   slot {class=WidgetSwitcherSlot}
  }
  Overlay_1 {widgetClassPath=&Overlay, parent=SizeBox_1, widget=Overlay_1, slot {class=SizeBoxSlot}}
  EntryBox_InfoOverlay {
   widgetClassPath=/Script/FortniteUI.FortDynamicEntryBox
   bInherited=true
   parent=Overlay_1
   widget=EntryBox_InfoOverlay
   properties {
    entryBoxType=Vertical
    entryWidgetClass=/Game/Athena/Collections/BP_AthenaCollectionLocationEntry_Character.BP_AthenaCollectionLocationEntry_Character_C
    maximumEntries=3
    entrySpacing=<0 6>
   }
   slot {class=OverlaySlot, properties {horizontalAlignment=HAlign_Fill}}
  }
  SizeBox_AllLocation {
   widgetClassPath=&SizeBox
   parent=Switcher_Locations
   widget=SizeBox_AllLocation
   properties {bOverride_MaxDesiredHeight=true, heightOverride=299.0475, maxDesiredHeight=361}
   slot {class=WidgetSwitcherSlot}
  }
  ScrollBox_Leaderboard {
   widgetClassPath=&ScrollBox
   parent=SizeBox_AllLocation
   widget=ScrollBox_Leaderboard
   properties {
    scrollbarPadding=<ltrb 6 0 0 0>
    scrollbarThickness=<4 4>
    widgetBarStyle {
     thickness=16
     draggedThumbImage {
      bIsDynamicallyLoaded=false
      drawAs=Image
      imageType=FullColor
      mirroring=NoMirror
      resourceName=../../../Engine/Content/Slate/Common/Scrollbar_Thumb.png
      resourceObject=/Game/Athena/UI/ReplayBrowser/M_Ui_Skewed_Vfor9slices_BG_Separator_Inst
      tiling=NoTile
      imageSize=<0 0>
      margin=<ltrb 0.25 0.25 0.25 0.25>
      outlineSettings {
       bUseBrushTransparency=false
       roundingType=HalfHeightRadius
       width=0
       color {colorUseRule=UseColor_Specified, specifiedColor=<rgba 0 0 0 0>}
       cornerRadii=<0 0 0 0>
      }
      tintColor {colorUseRule=UseColor_Specified, specifiedColor=<rgba 1 1 1 1>}
      uVRegion {bIsValid=false, max=<0 0>, min=<0 0>}
     }
     horizontalBackgroundImage {
      bIsDynamicallyLoaded=false
      drawAs=NoDrawType
      imageType=NoImage
      mirroring=NoMirror
      resourceName=None
      resourceObject=null
      tiling=NoTile
      imageSize=<0 0>
      margin=<ltrb 0 0 0 0>
      outlineSettings {
       bUseBrushTransparency=false
       roundingType=HalfHeightRadius
       width=0
       color {colorUseRule=UseColor_Specified, specifiedColor=<rgba 0 0 0 0>}
       cornerRadii=<0 0 0 0>
      }
      tintColor {colorUseRule=UseColor_Specified, specifiedColor=<rgba 1 1 1 1>}
      uVRegion {bIsValid=false, max=<0 0>, min=<0 0>}
     }
     horizontalBottomSlotImage {
      bIsDynamicallyLoaded=false
      drawAs=NoDrawType
      imageType=FullColor
      mirroring=NoMirror
      resourceName=../../../Engine/Content/Slate/Common/Scrollbar_Background_Horizontal.png
      resourceObject=null
      tiling=NoTile
      imageSize=<8 8>
      margin=<ltrb 0 0 0 0>
      outlineSettings {
       bUseBrushTransparency=false
       roundingType=HalfHeightRadius
       width=0
       color {colorUseRule=UseColor_Specified, specifiedColor=<rgba 0 0 0 0>}
       cornerRadii=<0 0 0 0>
      }
      tintColor {colorUseRule=UseColor_Specified, specifiedColor=<rgba 1 1 1 1>}
      uVRegion {bIsValid=false, max=<0 0>, min=<0 0>}
     }
     horizontalTopSlotImage {
      bIsDynamicallyLoaded=false
      drawAs=NoDrawType
      imageType=FullColor
      mirroring=NoMirror
      resourceName=../../../Engine/Content/Slate/Common/Scrollbar_Background_Horizontal.png
      resourceObject=null
      tiling=NoTile
      imageSize=<8 8>
      margin=<ltrb 0 0 0 0>
      outlineSettings {
       bUseBrushTransparency=false
       roundingType=HalfHeightRadius
       width=0
       color {colorUseRule=UseColor_Specified, specifiedColor=<rgba 0 0 0 0>}
       cornerRadii=<0 0 0 0>
      }
      tintColor {colorUseRule=UseColor_Specified, specifiedColor=<rgba 1 1 1 1>}
      uVRegion {bIsValid=false, max=<0 0>, min=<0 0>}
     }
     hoveredThumbImage {
      bIsDynamicallyLoaded=false
      drawAs=Image
      imageType=FullColor
      mirroring=NoMirror
      resourceName=../../../Engine/Content/Slate/Common/Scrollbar_Thumb.png
      resourceObject=&M_Ui_Skewed_Vfor9slices_BG_Separator_Inst
      tiling=NoTile
      imageSize=<0 0>
      margin=<ltrb 0.25 0.25 0.25 0.25>
      outlineSettings {
       bUseBrushTransparency=false
       roundingType=HalfHeightRadius
       width=0
       color {colorUseRule=UseColor_Specified, specifiedColor=<rgba 0 0 0 0>}
       cornerRadii=<0 0 0 0>
      }
      tintColor {colorUseRule=UseColor_Specified, specifiedColor=<rgba 0.5573 0.808 1 1>}
      uVRegion {bIsValid=false, max=<0 0>, min=<0 0>}
     }
     normalThumbImage {
      bIsDynamicallyLoaded=false
      drawAs=Image
      imageType=FullColor
      mirroring=NoMirror
      resourceName=../../../Engine/Content/Slate/Common/Scrollbar_Thumb.png
      resourceObject=&M_Ui_Skewed_Vfor9slices_BG_Separator_Inst
      tiling=NoTile
      imageSize=<0 0>
      margin=<ltrb 0.25 0.25 0.25 0.25>
      outlineSettings {
       bUseBrushTransparency=false
       roundingType=HalfHeightRadius
       width=0
       color {colorUseRule=UseColor_Specified, specifiedColor=<rgba 0 0 0 0>}
       cornerRadii=<0 0 0 0>
      }
      tintColor {colorUseRule=UseColor_Specified, specifiedColor=<rgba 0.2122 0.6584 1 1>}
      uVRegion {bIsValid=false, max=<0 0>, min=<0 0>}
     }
     verticalBackgroundImage {
      bIsDynamicallyLoaded=false
      drawAs=Image
      imageType=NoImage
      mirroring=NoMirror
      resourceName=None
      resourceObject=&M_Ui_Skewed_Vfor9slices_BG_Separator_Inst
      tiling=NoTile
      imageSize=<0 0>
      margin=<ltrb 0 0 0 0>
      outlineSettings {
       bUseBrushTransparency=false
       roundingType=HalfHeightRadius
       width=0
       color {colorUseRule=UseColor_Specified, specifiedColor=<rgba 0 0 0 0>}
       cornerRadii=<0 0 0 0>
      }
      tintColor {colorUseRule=UseColor_Specified, specifiedColor=<rgba 0 0.2159 1 0.498>}
      uVRegion {bIsValid=false, max=<0 0>, min=<0 0>}
     }
     verticalBottomSlotImage {
      bIsDynamicallyLoaded=false
      drawAs=NoDrawType
      imageType=FullColor
      mirroring=NoMirror
      resourceName=../../../Engine/Content/Slate/Common/Scrollbar_Background_Vertical.png
      resourceObject=null
      tiling=NoTile
      imageSize=<8 8>
      margin=<ltrb 0 0 0 0>
      outlineSettings {
       bUseBrushTransparency=false
       roundingType=HalfHeightRadius
       width=0
       color {colorUseRule=UseColor_Specified, specifiedColor=<rgba 0 0 0 0>}
       cornerRadii=<0 0 0 0>
      }
      tintColor {colorUseRule=UseColor_Specified, specifiedColor=<rgba 1 1 1 1>}
      uVRegion {bIsValid=false, max=<0 0>, min=<0 0>}
     }
     verticalTopSlotImage {
      bIsDynamicallyLoaded=false
      drawAs=NoDrawType
      imageType=FullColor
      mirroring=NoMirror
      resourceName=../../../Engine/Content/Slate/Common/Scrollbar_Background_Vertical.png
      resourceObject=null
      tiling=NoTile
      imageSize=<8 8>
      margin=<ltrb 0 0 0 0>
      outlineSettings {
       bUseBrushTransparency=false
       roundingType=HalfHeightRadius
       width=0
       color {colorUseRule=UseColor_Specified, specifiedColor=<rgba 0 0 0 0>}
       cornerRadii=<0 0 0 0>
      }
      tintColor {colorUseRule=UseColor_Specified, specifiedColor=<rgba 1 1 1 1>}
      uVRegion {bIsValid=false, max=<0 0>, min=<0 0>}
     }
    }
   }
   slot {class=SizeBoxSlot}
  }
  EntryBox_Leaderboard {
   widgetClassPath=&FortDynamicEntryBox
   parent=ScrollBox_Leaderboard
   widget=EntryBox_Leaderboard
   properties {
    entryBoxType=Vertical
    entryWidgetClass=&BP_AthenaCollectionLocationEntry_Character_C
    numDesignerPreviewEntries=13
    entrySpacing=<0 4>
   }
   slot {class=ScrollBoxSlot}
  }
  HorizontalBox_49 {
   widgetClassPath=&HorizontalBox
   parent=VerticalBox_22
   widget=HorizontalBox_49
   slot {class=VerticalBoxSlot, properties {horizontalAlignment=HAlign_Right, padding=<ltrb 0 16 0 0>}}
  }
  Overlay_locationsButton {widgetClassPath=&Overlay, parent=HorizontalBox_49, widget=Overlay_locationsButton, slot {class=HorizontalBoxSlot}}
  HardwareVisibilityBorder_Close {
   widgetClassPath=&CommonHardwareVisibilityBorder
   parent=HorizontalBox_49
   widget=HardwareVisibilityBorder_Close
   properties {visibilityQuery {
     autoDescription=" NONE( Input.Touch )"
     tokenStreamVersion=0
     userDescription=""
     queryTokenStream [0,1,3,1,0]
     tagDictionary [{tagName=Input.Touch}]
    }}
   slot {class=HorizontalBoxSlot, properties {padding=<ltrb 32 0 0 0>}}
  }
  Button_Close {
   widgetClassPath=/FortUILibrary/UIKit/Buttons/WBP_UIKit_Button_Quiet
   bInherited=true
   bIsVariable=true
   parent=HardwareVisibilityBorder_Close
   widget=Button_Close
   properties {minWidth=120, text=Exit, triggeringInputAction {dataTable=&InputActionDataTable, rowName=GenericBack}}
   slot {class=BorderSlot, properties {padding=<ltrb 0 0 0 0>}}
  }
  RichText_NumCollected {widgetClassPath=&CommonRichTextBlock, bInherited=true, widget=null}
  FriendLeaderboard {widgetClassPath=/Script/FortniteUI.AthenaCollectionLeaderboard, bInherited=true, widget=null}
 ]
}
